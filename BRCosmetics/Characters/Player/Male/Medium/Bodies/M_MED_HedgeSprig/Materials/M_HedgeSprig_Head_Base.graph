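# Cloth-shaded head material for the HedgeSprig player character, adding animated forehead wrinkles to the 2023 anime shader.
# Nested Ifs threshold glabella and brow curve parameters to pick a pose, with Brow Up forced as default.
# The pose drives a flipbook in scaled forehead UVs, merged by Max with the wrinkle-masked DFL.
# DFL is intercepted between the shader's Pre and Calculations functions, before Post shading, using named reroutes.
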Material /BRCosmetics/Characters/Player/Male/Medium/Bodies/M_MED_HedgeSprig/Materials/M_HedgeSprig_Head_Base
graph {
 properties {
  bAutomaticallySetUsageInEditor=false
  bUseMaterialAttributes=true
  bUsedWithMeshDeformer=true
  bUsedWithMorphTargets=true
  bUsedWithSkeletalMesh=true
  bUsedWithStaticLighting=true
  pixelDepthOffsetMode=PDOM_Legacy
  shadingModel=MSM_Cloth
  usedShadingModels=Cloth
 }
 nodes [
  Constant()
  n1 = Multiply(A=NamedRerouteUsage()) {constB=2, desc=FORT-XXXX - Forcing the Brow Up pose to be the default pose.}
  n2 = If(A=NamedRerouteUsage(), B=NamedRerouteUsage(), "A > B"=NamedRerouteUsage(), "A == B"=@n1, "A < B"=@n1)
  n3 = If(A=NamedRerouteUsage(), B=NamedRerouteUsage(), "A > B"=Multiply(A=NamedRerouteUsage()) {constB=2}, "A == B"=@n2, "A < B"=@n2)
  If(A=NamedRerouteUsage(), B=NamedRerouteUsage(), "A > B"=Multiply(A=NamedRerouteUsage()) {constB=3}, "A == B"=@n3, "A < B"=@n3)
  n4 = MaterialFunctionCall!(CustomUV) {materialFunction=/Game/Global/Materials/Character_Parent_2023/Anime_Parent_2023/MaterialFunctions/MFC_AnimeShader_2023_Pre}
  n5 = MaterialFunctionCall!(CustomLightVector, OverrideNormal_WS, STT=@n4.STT, Normal_Map=@n4.STT, DFL=NamedRerouteUsage(), FX=@n4.STT) {materialFunction=/Game/Global/Materials/Character_Parent_2023/Anime_Parent_2023/MaterialFunctions/MFC_AnimeShader_2023_Calculations}
  MaterialFunctionCall!(
   Shading_Gradient=@n5.Shading_Gradient
   Threshold_Gradient=@n5.Shading_Gradient
   Specular_Gradient=@n5.Shading_Gradient
   Sheen_Mask=@n5.Shading_Gradient
   Thickness=@n5.Shading_Gradient
   DFL_Mask=@n5.Shading_Gradient
   FX=@n5.Shading_Gradient
   Metallic_Gradient=@n5.Shading_Gradient
   RimLight_Gradient=@n5.Shading_Gradient
   NormalBuffer=@n5.Shading_Gradient
   --------
   CINE_LightSources=@n5.Shading_Gradient
   CINE_LightColor=@n5.Shading_Gradient
   ---------
   Color_Lit=@n4.Color_Lit
   Color_Shaded=@n4.Color_Lit
   SRM=@n4.Color_Lit
   Emissive=@n4.Color_Lit
   DFL_Color=@n4.Color_Lit
   Specular_Color=@n4.Color_Lit
   RimLight_Color=@n4.Color_Lit
  ) {materialFunction=/Game/Global/Materials/Character_Parent_2023/Anime_Parent_2023/MaterialFunctions/MFC_AnimeShader_2023_Post}
  NamedRerouteDeclaration(None=@n4.DFL) {name=IN_DFL, nodeColor=<rgba 0.7578 1 0 1>}
  n6 = TextureSampleParameter2D() {
   group=HedgeSprig Facial Animation
   parameterName=WrinkleMask
   texture=/BRCosmetics/Characters/Player/Male/Medium/Bodies/M_MED_HedgeSprig/Textures/T_HedgeSprig_Head_Wrinkle_Mask
  }
  n7 = Subtract(A=NamedRerouteUsage(), B=@n6.RGB)
  n8 = NamedRerouteDeclaration(None=ScalarParameter() {defaultValue=2, group=HedgeSprig Facial Animation, parameterName=NumOfRowsAndColumns}) {name=NumOfRowsAndColumns, nodeColor=<rgba 1 0.6328 0 1>}
  n9 = TextureObjectParameter() {
   group=HedgeSprig Facial Animation
   parameterName=WrinkleFlipbook
   samplerType=SAMPLERTYPE_Alpha
   texture=/BRCosmetics/Characters/Player/Male/Medium/Bodies/M_MED_HedgeSprig/Textures/T_HedgeSprig_Forehead_DFL
  }
  n10 = MaterialFunctionCall(UVs=Add(A=TextureCoordinate(), B=Constant2Vector() {g=0.22}), Texture Scale=Constant2Vector() {g=0.1, r=0.15}) {materialFunction=&ScaleUVsByCenter}
  n11 = MaterialFunctionCall("Animation  Phase (0-1)"=NamedRerouteUsage(), Number of Rows=@n8, Number of Columns=@n8, Texture=@n9, UVs=@n10.UVs) {materialFunction=&FlipBook}
  NamedRerouteDeclaration(None=Max(A=Saturate(None=@n7), B=Multiply(A=@n11.Result, B=@n10.0-1 mask))) {name=OUT_DFL, nodeColor=<rgba 1 0.6328 0 1>}
  n12 = NamedRerouteUsage()
  NamedRerouteDeclaration(None=Divide(B=Multiply(A=@n12, B=@n12))) {name=AnimationStep, nodeColor=<rgba 0.3828 0 1 1>}
  NamedRerouteDeclaration(None=@n3) {name=FinalPose, nodeColor=<rgba 1 0.1406 0 1>}
  NamedRerouteDeclaration(None=ScalarParameter() {group=HedgeSprig Animation Curves, parameterName=C_glabella_down_pose}) {name=GlabellaDownPose, nodeColor=<rgba 1 0 0.5625 1>}
  NamedRerouteDeclaration(None=ScalarParameter() {group=HedgeSprig Animation Curves, parameterName=C_glabella_up_pose}) {name=GlabellaUpPose, nodeColor=<rgba 0 0.8828 1 1>}
  NamedRerouteDeclaration(None=ScalarParameter() {group=HedgeSprig Animation Curves, parameterName=R_brow_up_pose}) {name=BrowUpPose, nodeColor=<rgba 1 0.7969 0 1>}
  NamedRerouteDeclaration(None=ScalarParameter() {defaultValue=0.1, group=HedgeSprig Facial Animation, parameterName=glabella_down_threshold}) {name=GlabellaDownThreshold, nodeColor=<rgba 0.4766 0 1 1>}
  NamedRerouteDeclaration(None=ScalarParameter() {defaultValue=0.1, group=HedgeSprig Facial Animation, parameterName=glabella_up_threshold}) {name=GlabellaUpThreshold, nodeColor=<rgba 0 1 0.1562 1>}
  NamedRerouteDeclaration(None=ScalarParameter() {defaultValue=0.1, group=HedgeSprig Facial Animation, parameterName=brow_up_threshold}) {name=BrowUpThreshold, nodeColor=<rgba 1 0 0.1641 1>}
 ]
}
